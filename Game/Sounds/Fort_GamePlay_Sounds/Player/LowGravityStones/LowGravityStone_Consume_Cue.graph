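# Sound cue played when the player consumes a Low Gravity Stone, per its name and folder.
# A single wave player feeds the LowGravityStone_Collect_07 wave straight to the output, with no randomization or layering.
# It uses the shared ability status-effect activation attenuation and routes through the gadgets and items sound class.
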SoundCue /Game/Sounds/Fort_GamePlay_Sounds/Player/LowGravityStones/LowGravityStone_Consume_Cue
root=WavePlayer_0
nodes [{id=WavePlayer_0, properties {soundWaveAssetPtr=&/LowGravityStone_Collect_07}}]
properties {
 attenuationSettings=/Game/Sounds/Attenuation/Fort_Ability_Status_Effect_Activation_Att
 maxDistance=6000
 soundClassObject=&Fort_Gadgets_Items
 volumeMultiplier=1
}
