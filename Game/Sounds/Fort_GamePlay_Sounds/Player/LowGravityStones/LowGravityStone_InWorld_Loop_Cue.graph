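# Sound cue for the looping in-world ambience of a Low Gravity Stone, routed to the gadgets/items sound class.
# A looping wave feeds an indefinitely looping enveloper whose pitch and volume curves slowly dip and swell over a 20-second cycle to vary the sound.
# All quality levels share that same branch.
# It uses forged-consumable attenuation and overrides concurrency to cap simultaneous instances, stopping the quietest.
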SoundCue /Game/Sounds/Fort_GamePlay_Sounds/Player/LowGravityStones/LowGravityStone_InWorld_Loop_Cue
root=QualityLevel_0
nodes [
 {
  id=Enveloper_0
  children [WavePlayer_0]
  properties {
   bLoop=true
   bLoopIndefinitely=true
   loopEnd=20
   pitchCurve {editorCurveData {keys [{time=0}, <at 10 0.8>, <at 15 1.1>, {time=20}]}}
   volumeCurve {editorCurveData {keys [<at 0 1>,<at 10 0.75>,<at 15 1>]}}
  }
 }
 {id=QualityLevel_0, children [Enveloper_0,Enveloper_0,Enveloper_0,Enveloper_0]}
 {
  id=WavePlayer_0
  properties {bLooping=true, soundWaveAssetPtr=&/LowGravityStone_InWorld_Loop}
 }
]
properties {
 attenuationSettings=/Game/Sounds/Attenuation/Fort_ForgedConsumable_InWorld_Att
 bOverrideConcurrency=true
 maxDistance=1000
 soundClassObject=&Fort_Gadgets_Items
 volumeMultiplier=0.5
 concurrencyOverrides {maxCount=4, resolutionRule=StopQuietest}
}
